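# Physics asset for the dynamic right-arm band bones on the male medium Cardboard Crew player body cosmetic.
# Two matching capsules on dyn_band_r_01 and dyn_band_r_02 each attach by their own constraint to a small capsule on dyn_band_r_MSTR, branching from it rather than forming a chain.
# Both constraints set moderate swing limits and tight twist limits, so each band piece can swing but barely rotate about its axis.
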PhysicsAsset /BRCosmetics/Characters/Player/Male/Medium/Bodies/M_MED_Cardboard_Crew/Meshes/M_MED_Cardboard_Crew_Dyn_arm_r
bodies {
 dyn_band_r_01 [{shapeType=Capsule, length=2.2404, radius=0.7593, center=<1.5 -0.0002 0>, rotation=<rot 90 0 0>}]
 dyn_band_r_02 [{shapeType=Capsule, length=2.2404, radius=0.7593, center=<1.5 -0.0002 0>, rotation=<rot 90 0 0>}]
 dyn_band_r_MSTR [{shapeType=Capsule, length=0.5, radius=0.5}]
}
constraints [
 {bone1Name=dyn_band_r_01, bone2Name=dyn_band_r_MSTR, swing1LimitDegrees=45, swing2LimitDegrees=45, twistLimitDegrees=5}
 {bone1Name=dyn_band_r_02, bone2Name=dyn_band_r_MSTR, swing1LimitDegrees=45, swing2LimitDegrees=45, twistLimitDegrees=5}
]
