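# Physics asset for the left leg of the male Cardboard Crew character, setting up dynamic simulation of its inner and outer shoelaces.
# Each lace is a capsule chain of a master root and two segments, joined with limited swing and tight twist.
# Both master roots attach to the left calf twist bone with free angular motion.
# Larger capsules on the calf twist and foot bones provide body shapes around the laces.
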PhysicsAsset /BRCosmetics/Characters/Player/Male/Medium/Bodies/M_MED_Cardboard_Crew/Meshes/M_MED_Cardboard_Crew_Dyn_leg_l
bodies {
 calf_twist_01_l [{shapeType=Capsule, length=16.5882, radius=4.2532, center=<-1.5778 0.6594 0>, rotation=<rot 90 0 0>}]
 dyn_shoelace_in_l_01 [{shapeType=Capsule, length=2.2581, radius=0.5589, center=<2.4339 0 0.0001>, rotation=<rot 90 0 0>}]
 dyn_shoelace_in_l_02 [{shapeType=Capsule, length=4.4411, radius=0.5589, center=<1.699 0 0.0001>, rotation=<rot 90 0 0>}]
 dyn_shoelace_in_l_MSTR [{shapeType=Capsule, length=0.5, radius=0.5}]
 dyn_shoelace_out_l_01 [{shapeType=Capsule, length=2.2581, radius=0.5589, center=<2.4339 0 0.0001>, rotation=<rot 90 0 0>}]
 dyn_shoelace_out_l_02 [{shapeType=Capsule, length=4.4411, radius=0.5589, center=<2.6022 -0.8038 0.1803>, rotation=<rot 67.7314 64.4957 82.7276>}]
 dyn_shoelace_out_l_MSTR [{shapeType=Capsule, length=0.5, radius=0.5}]
 foot_l [{shapeType=Capsule, length=16.5882, radius=4.2532, center=<-1.5778 6.9432 -0.0001>, rotation=<rot 0.0004 24.9991 89.9992>}]
}
constraints [
 {bone1Name=dyn_shoelace_out_l_01, bone2Name=dyn_shoelace_out_l_MSTR, swing1LimitDegrees=45, swing2LimitDegrees=45, twistLimitDegrees=5}
 {bone1Name=dyn_shoelace_out_l_02, bone2Name=dyn_shoelace_out_l_01, swing1LimitDegrees=45, swing2LimitDegrees=45, twistLimitDegrees=5}
 {bone1Name=dyn_shoelace_in_l_01, bone2Name=dyn_shoelace_in_l_MSTR, swing1LimitDegrees=45, swing2LimitDegrees=45, twistLimitDegrees=5}
 {bone1Name=dyn_shoelace_in_l_02, bone2Name=dyn_shoelace_in_l_01, swing1LimitDegrees=45, swing2LimitDegrees=45, twistLimitDegrees=5}
 {
  bone1Name=dyn_shoelace_out_l_MSTR
  bone2Name=calf_twist_01_l
  swing1LimitDegrees=45
  swing1Motion=Free
  swing2LimitDegrees=45
  swing2Motion=Free
  twistLimitDegrees=45
  twistMotion=Free
 }
 {
  bone1Name=dyn_shoelace_in_l_MSTR
  bone2Name=calf_twist_01_l
  swing1LimitDegrees=45
  swing1Motion=Free
  swing2LimitDegrees=45
  swing2Motion=Free
  twistLimitDegrees=45
  twistMotion=Free
 }
]
