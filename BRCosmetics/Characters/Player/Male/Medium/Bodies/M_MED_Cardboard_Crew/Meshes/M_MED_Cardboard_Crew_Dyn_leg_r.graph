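# Physics asset for the right leg of the male medium Cardboard Crew character body, covering its dynamic inner and outer shoelace bones.
# Each lace is a three-body capsule chain (MSTR, 01, 02) whose master joins calf_twist_01_r through an angularly free constraint.
# Links within each chain have swing limit angles and a narrow twist limit angle.
# Larger capsules cover the calf and foot, with the foot body left unconstrained.
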PhysicsAsset /BRCosmetics/Characters/Player/Male/Medium/Bodies/M_MED_Cardboard_Crew/Meshes/M_MED_Cardboard_Crew_Dyn_leg_r
bodies {
 calf_twist_01_r [{shapeType=Capsule, length=16.5882, radius=4.2532, center=<1.5778 -0.6594 0>, rotation=<rot 90 0 0>}]
 dyn_shoelace_in_r_01 [{shapeType=Capsule, length=2.2581, radius=0.5589, center=<2.4339 0 0.0001>, rotation=<rot 90 0 0>}]
 dyn_shoelace_in_r_02 [{shapeType=Capsule, length=4.4411, radius=0.5589, center=<3.3292 -0.814 -0.3546>, rotation=<rot 80 -9.9998 0>}]
 dyn_shoelace_in_r_MSTR [{shapeType=Capsule, length=0.5, radius=0.5}]
 dyn_shoelace_out_r_01 [{shapeType=Capsule, length=2.2581, radius=0.5589, center=<2.1805 0 0.0776>, rotation=<rot 90 0 0>}]
 dyn_shoelace_out_r_02 [{shapeType=Capsule, length=4.4411, radius=0.5589, center=<2.3758 0.3805 0.0659>, rotation=<rot 90 -179.39 170.6101>}]
 dyn_shoelace_out_r_MSTR [{shapeType=Capsule, length=0.5, radius=0.5}]
 foot_r [{shapeType=Capsule, length=16.5882, radius=4.2532, center=<1.5778 -6.9432 0.0001>, rotation=<rot 0.0004 24.9991 89.9992>}]
}
constraints [
 {bone1Name=dyn_shoelace_out_r_01, bone2Name=dyn_shoelace_out_r_MSTR, swing1LimitDegrees=45, swing2LimitDegrees=45, twistLimitDegrees=5}
 {bone1Name=dyn_shoelace_out_r_02, bone2Name=dyn_shoelace_out_r_01, swing1LimitDegrees=45, swing2LimitDegrees=45, twistLimitDegrees=5}
 {bone1Name=dyn_shoelace_in_r_01, bone2Name=dyn_shoelace_in_r_MSTR, swing1LimitDegrees=45, swing2LimitDegrees=45, twistLimitDegrees=5}
 {bone1Name=dyn_shoelace_in_r_02, bone2Name=dyn_shoelace_in_r_01, swing1LimitDegrees=45, swing2LimitDegrees=45, twistLimitDegrees=5}
 {
  bone1Name=dyn_shoelace_out_r_MSTR
  bone2Name=calf_twist_01_r
  swing1LimitDegrees=45
  swing1Motion=Free
  swing2LimitDegrees=45
  swing2Motion=Free
  twistLimitDegrees=45
  twistMotion=Free
 }
 {
  bone1Name=dyn_shoelace_in_r_MSTR
  bone2Name=calf_twist_01_r
  swing1LimitDegrees=45
  swing1Motion=Free
  swing2LimitDegrees=45
  swing2Motion=Free
  twistLimitDegrees=45
  twistMotion=Free
 }
]
