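# Physics asset for the Cardboard Crew male medium body's torso, simulating the lower panels and flaps of its cardboard box.
# Box-shaped flaps hinge on four side panels through single-swing constraints.
# Panels join small anchor bodies linked to spine_03; back and right panels swing on one axis, while front and left also allow small secondary swing and twist.
# Pelvis, spine and thigh capsules, joined by free constraints, approximate the body.
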PhysicsAsset /BRCosmetics/Characters/Player/Male/Medium/Bodies/M_MED_Cardboard_Crew/Meshes/M_MED_Cardboard_Crew_Dyn_torso
bodies {
 dyn_torso_box_bott_bk_MSTR [{shapeType=Box, extentX=11.7101, extentY=24.8077, extentZ=2.8472, center=<4.8 0 0>}]
 dyn_torso_box_bott_bk_MSTR1 [{shapeType=Capsule, length=0.505, radius=0.505}]
 dyn_torso_box_bott_flap_bk [{shapeType=Box, extentX=8.5177, extentY=24.8077, extentZ=2.4722, center=<3.8024 0.1088 -1.3141>, rotation=<rot -0.212 -0.0649 -1.7679>}]
 dyn_torso_box_bott_flap_fr [{shapeType=Box, extentX=6.9192, extentY=28.0487, extentZ=1.5347, center=<4.3818 0.0167 0.2672>, rotation=<rot 0 -3.1908 0>}]
 dyn_torso_box_bott_flap_l [{shapeType=Box, extentX=8.5177, extentY=24.8077, extentZ=1.5347, center=<4.6802 0 -0.0001>}]
 dyn_torso_box_bott_flap_r [{shapeType=Box, extentX=8.5177, extentY=24.8077, extentZ=1.5347, center=<4.6802 0 -0.0001>}]
 dyn_torso_box_bott_fr_MSTR [{shapeType=Box, extentX=8.3779, extentY=25.8886, extentZ=1.5347, center=<7.3275 -0.0004 0>}]
 dyn_torso_box_bott_fr_MSTR1 [{shapeType=Capsule, length=0.505, radius=0.505}]
 dyn_torso_box_bott_l_MSTR [{shapeType=Box, extentX=29.3137, extentY=24.8077, extentZ=1.5347, center=<14.1471 -0.4261 0.6667>, rotation=<rot 7.8101 -3.9777 -5.4819>}]
 dyn_torso_box_bott_l_MSTR1 [{shapeType=Capsule, length=0.505, radius=0.505}]
 dyn_torso_box_bott_r_MSTR [{shapeType=Box, extentX=18.9094, extentY=24.8077, extentZ=1.5347, center=<9.0209 0.0001 0.7907>, rotation=<rot 5 0 0>}]
 dyn_torso_box_bott_r_MSTR1 [{shapeType=Capsule, length=0.505, radius=0.505}]
 pelvis [{shapeType=Capsule, length=4.6198, radius=12.8664, center=<-1.3506 -1.4738 0.0001>}]
 spine_01 [{shapeType=Capsule, length=4.336, radius=11.2378, center=<1.0152 -0.4381 0>}]
 spine_02 [{shapeType=Capsule, length=4.3961, radius=10.9557, center=<0.7574 -1.0776 0>}]
 spine_03 [{shapeType=Capsule, length=0.5, radius=0.5}]
 thigh_l [{shapeType=Capsule, length=37.2403, radius=8.0385, center=<-17.728 0 0>, rotation=<rot 87.5182 -179.9999 -179.9999>}]
 thigh_r [{shapeType=Capsule, length=37.2403, radius=8.0385, center=<17.728 0 0>, rotation=<rot 90 0 0>}]
}
constraints [
 {
  bone1Name=dyn_torso_box_bott_flap_l
  bone2Name=dyn_torso_box_bott_l_MSTR
  swing1LimitDegrees=45
  swing1Motion=Locked
  swing2LimitDegrees=45
  twistLimitDegrees=45
  twistMotion=Locked
 }
 {
  bone1Name=dyn_torso_box_bott_flap_fr
  bone2Name=dyn_torso_box_bott_fr_MSTR
  swing1LimitDegrees=45
  swing1Motion=Locked
  swing2LimitDegrees=45
  twistLimitDegrees=45
  twistMotion=Locked
 }
 {
  bone1Name=dyn_torso_box_bott_flap_r
  bone2Name=dyn_torso_box_bott_r_MSTR
  swing1LimitDegrees=45
  swing1Motion=Locked
  swing2LimitDegrees=45
  twistLimitDegrees=45
  twistMotion=Locked
 }
 {
  bone1Name=dyn_torso_box_bott_flap_bk
  bone2Name=dyn_torso_box_bott_bk_MSTR
  swing1LimitDegrees=45
  swing1Motion=Locked
  swing2LimitDegrees=45
  twistLimitDegrees=45
  twistMotion=Locked
 }
 {bone1Name=dyn_torso_box_bott_l_MSTR, bone2Name=dyn_torso_box_bott_l_MSTR1, swing1LimitDegrees=5, swing2LimitDegrees=45, twistLimitDegrees=10}
 {
  bone1Name=dyn_torso_box_bott_bk_MSTR
  bone2Name=dyn_torso_box_bott_bk_MSTR1
  swing1LimitDegrees=5
  swing1Motion=Locked
  swing2LimitDegrees=45
  twistLimitDegrees=5
  twistMotion=Locked
 }
 {
  bone1Name=dyn_torso_box_bott_fr_MSTR
  bone2Name=dyn_torso_box_bott_fr_MSTR1
  swing1LimitDegrees=5
  swing2LimitDegrees=45
  twistLimitDegrees=10
 }
 {
  bone1Name=dyn_torso_box_bott_r_MSTR
  bone2Name=dyn_torso_box_bott_r_MSTR1
  swing1LimitDegrees=5
  swing1Motion=Locked
  swing2LimitDegrees=45
  twistLimitDegrees=5
  twistMotion=Locked
 }
 {
  bone1Name=spine_03
  bone2Name=spine_02
  swing1LimitDegrees=45
  swing1Motion=Free
  swing2LimitDegrees=45
  swing2Motion=Free
  twistLimitDegrees=45
  twistMotion=Free
 }
 {
  bone1Name=thigh_l
  bone2Name=pelvis
  swing1LimitDegrees=45
  swing1Motion=Free
  swing2LimitDegrees=45
  swing2Motion=Free
  twistLimitDegrees=45
  twistMotion=Free
 }
 {
  bone1Name=thigh_r
  bone2Name=pelvis
  swing1LimitDegrees=45
  swing1Motion=Free
  swing2LimitDegrees=45
  swing2Motion=Free
  twistLimitDegrees=45
  twistMotion=Free
 }
 {
  bone1Name=spine_01
  bone2Name=pelvis
  swing1LimitDegrees=45
  swing1Motion=Free
  swing2LimitDegrees=45
  swing2Motion=Free
  twistLimitDegrees=45
  twistMotion=Free
 }
 {
  bone1Name=spine_02
  bone2Name=spine_01
  swing1LimitDegrees=45
  swing1Motion=Free
  swing2LimitDegrees=45
  swing2Motion=Free
  twistLimitDegrees=45
  twistMotion=Free
 }
 {
  bone1Name=dyn_torso_box_bott_fr_MSTR1
  bone2Name=spine_03
  swing1LimitDegrees=45
  swing1Motion=Free
  swing2LimitDegrees=45
  swing2Motion=Free
  twistLimitDegrees=45
  twistMotion=Free
 }
 {
  bone1Name=dyn_torso_box_bott_l_MSTR1
  bone2Name=spine_03
  swing1LimitDegrees=45
  swing1Motion=Free
  swing2LimitDegrees=45
  swing2Motion=Free
  twistLimitDegrees=45
  twistMotion=Free
 }
 {
  bone1Name=dyn_torso_box_bott_bk_MSTR1
  bone2Name=spine_03
  swing1LimitDegrees=45
  swing1Motion=Free
  swing2LimitDegrees=45
  swing2Motion=Free
  twistLimitDegrees=45
  twistMotion=Free
 }
 {
  bone1Name=dyn_torso_box_bott_r_MSTR1
  bone2Name=spine_03
  swing1LimitDegrees=45
  swing1Motion=Free
  swing2LimitDegrees=45
  swing2Motion=Free
  twistLimitDegrees=45
  twistMotion=Free
 }
]
