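# Post-process style AnimBP for the male Cardboard Crew body cosmetic, adding secondary motion to a copied parent-mesh pose.
# Constraints blend box-flap bones against spine and clavicles, then RigidBody nodes simulate torso, right arm and legs.
# Crouch state tilts the front flap, and frame time scales leg simulation alpha.
# Additive bone rotations adjust side and back flaps; spine-bend correction values are computed when rigid bodies are disabled or LOD exceeds 1.
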AnimBlueprint /BRCosmetics/Characters/Player/Male/Medium/Bodies/M_MED_Cardboard_Crew/Meshes/M_MED_Cardboard_Crew_AnimBP
target_skeleton=&/M_MED_Cardboard_Crew_Skeleton
graphs [
 AnimGraph {nodes [
   n1 = "Constraint - dyn_torso_box_bott_r_MSTR1 (PPPPP)"!(
    ComponentPose=Local To Component(LocalPose=Copy Pose From Mesh() {bUseAttachedParent=true})
    Alpha=1
    ConstraintWeights_0=0.2766
    ConstraintWeights_1:float=0.0138
    ConstraintWeights_2:float=0.3193
    ConstraintWeights_3:float=0.3547
    ConstraintWeights_4:float=0.0357
   ) -> (Pose) {
    alphaBoolBlend {blendOption=Linear}
    boneToModify {boneName=dyn_torso_box_bott_r_MSTR1}
    constraintSetup [
     {targetBone {boneName=dyn_torso_box_bott_r_MSTR}}
     {targetBone {boneName=clavicle_r}}
     {targetBone {boneName=spine_03}}
     {targetBone {boneName=spine_04}}
     {targetBone {boneName=spine_05}}
    ]
    constraintWeights [1,1,1,1,1]
   }
   n2 = "Constraint - dyn_torso_box_bott_fr_MSTR1 (PPPP)"!(
    ComponentPose=@n1
    Alpha=1
    ConstraintWeights_0=0.1969
    ConstraintWeights_1:float=0.2635
    ConstraintWeights_2:float=0.477
    ConstraintWeights_3:float=0.0625
   ) -> (Pose) {
    alphaBoolBlend {blendOption=Linear}
    boneToModify {boneName=dyn_torso_box_bott_fr_MSTR1}
    constraintSetup [
     {targetBone {boneName=spine_03}}
     {targetBone {boneName=spine_04}}
     {targetBone {boneName=dyn_torso_box_bott_fr_MSTR}}
     {targetBone {boneName=spine_05}}
    ]
    constraintWeights [1,1,1,1]
   }
   n3 = "Constraint - dyn_torso_box_bott_bk_MSTR1 (PPPP)"!(
    ComponentPose=@n2
    Alpha=1
    ConstraintWeights_0=0.3425
    ConstraintWeights_1:float=0.3328
    ConstraintWeights_2:float=0.1162
    ConstraintWeights_3:float=0.2085
   ) -> (Pose) {
    alphaBoolBlend {blendOption=Linear}
    boneToModify {boneName=dyn_torso_box_bott_bk_MSTR1}
    constraintSetup [
     {targetBone {boneName=spine_03}}
     {targetBone {boneName=spine_04}}
     {targetBone {boneName=spine_05}}
     {targetBone {boneName=dyn_torso_box_bott_bk_MSTR}}
    ]
    constraintWeights [1,1,1,1]
   }
   n4 = "Constraint - dyn_torso_box_bott_l_MSTR1 (PPPPP)"!(
    ComponentPose=@n3
    Alpha=1
    ConstraintWeights_0=0.3297
    ConstraintWeights_1:float=0.3797
    ConstraintWeights_2:float=0.0164
    ConstraintWeights_3:float=0.2323
    ConstraintWeights_4:float=0.0419
   ) -> (Pose) {
    alphaBoolBlend {blendOption=Linear}
    boneToModify {boneName=dyn_torso_box_bott_l_MSTR1}
    constraintSetup [
     {targetBone {boneName=spine_03}}
     {targetBone {boneName=spine_04}}
     {targetBone {boneName=clavicle_l}}
     {targetBone {boneName=dyn_torso_box_bott_l_MSTR}}
     {targetBone {boneName=spine_05}}
    ]
    constraintWeights [1,1,1,1,1]
   }
   n5 = "Transform (Modify) Bone - Bone: dyn_torso_box_bott_fr_MSTR1"(
    ComponentPose=@n4
    Alpha=Get crouch alpha() -> (crouch alpha:double)
    Translation="-3.5,0.000,-0.000"
    Rotation="100,0.000,0.000"
    Scale="1,1,1"
   ) {
    rotationMode=BMM_Additive
    rotationSpace=BCS_BoneSpace
    translationMode=BMM_Additive
    translationSpace=BCS_BoneSpace
    alphaBoolBlend {blendOption=Linear}
    alphaScaleBiasClamp {bInterpResult=true, interpSpeedDecreasing=20, interpSpeedIncreasing=20}
    boneToModify {boneName=dyn_torso_box_bott_fr_MSTR1}
    rotation {pitch=100}
    translation {x=-3.5}
   }
   n6 = RigidBody(ComponentPose=@n5) {
    bClampLinearTranslationLimitToRefPose=true
    bOverrideWorldGravity=true
    overridePhysicsAsset=&/M_MED_Cardboard_Crew_Dyn_torso
    simulationSpace=BaseBoneSpace
    alphaBoolBlend {blendOption=Linear}
    baseBoneRef {boneName=pelvis}
    componentAppliedLinearAccClamp=<3000 3000 3000>
    componentLinearAccScale=<0.4 0.3 0.3>
    componentLinearVelScale=<0.4 0.3 0.3>
    overrideWorldGravity {z=-980}
   }
   n7 = RigidBody(ComponentPose=@n6) {
    bClampLinearTranslationLimitToRefPose=true
    bOverrideWorldGravity=true
    overridePhysicsAsset=&/M_MED_Cardboard_Crew_Dyn_arm_r
    simulationSpace=BaseBoneSpace
    alphaBoolBlend {blendOption=Linear}
    baseBoneRef {boneName=upperarm_r}
    componentAppliedLinearAccClamp=<3000 3000 3000>
    componentLinearAccScale=<0.4 0.3 0.3>
    componentLinearVelScale=<0.4 0.3 0.3>
    overrideWorldGravity {z=-980}
   }
   n8 = Get rigid_body_alpha() -> (rigid_body_alpha:double)
   n9 = RigidBody(ComponentPose=@n7, Alpha=@n8) {
    bClampLinearTranslationLimitToRefPose=true
    bOverrideWorldGravity=true
    overridePhysicsAsset=&/M_MED_Cardboard_Crew_Dyn_leg_l
    simulationSpace=BaseBoneSpace
    alphaBoolBlend {blendOption=Linear}
    baseBoneRef {boneName=thigh_l}
    componentAppliedLinearAccClamp=<3000 3000 3000>
    componentLinearAccScale=<0.4 0.3 0.3>
    componentLinearVelScale=<0.4 0.3 0.3>
    overrideWorldGravity {z=-980}
   }
   n10 = RigidBody(ComponentPose=@n9, Alpha=@n8) {
    bClampLinearTranslationLimitToRefPose=true
    bOverrideWorldGravity=true
    overridePhysicsAsset=&/M_MED_Cardboard_Crew_Dyn_leg_r
    simulationSpace=BaseBoneSpace
    alphaBoolBlend {blendOption=Linear}
    baseBoneRef {boneName=thigh_r}
    componentAppliedLinearAccClamp=<3000 3000 3000>
    componentLinearAccScale=<0.4 0.3 0.3>
    componentLinearVelScale=<0.4 0.3 0.3>
    overrideWorldGravity {z=-980}
   }
   n11 = "Transform (Modify) Bone - Bone: dyn_torso_box_bott_l_MSTR"!(ComponentPose=@n10, Alpha=1, Rotation="20,0,0") -> (Pose) {rotationMode=BMM_Additive, rotationSpace=BCS_BoneSpace, boneToModify {boneName=dyn_torso_box_bott_l_MSTR}, rotation {pitch=20}}
   n12 = "Transform (Modify) Bone - Bone: dyn_torso_box_bott_r_MSTR"!(ComponentPose=@n11, Alpha=1, Rotation="20,0,0") -> (Pose) {rotationMode=BMM_Additive, rotationSpace=BCS_BoneSpace, boneToModify {boneName=dyn_torso_box_bott_r_MSTR}, rotation {pitch=20}}
   n13 = "Transform (Modify) Bone - Bone: dyn_torso_box_bott_bk_MSTR"!(ComponentPose=@n12, Alpha=1, Rotation="32,0,0") -> (Pose) {
    rotationMode=BMM_Additive
    rotationSpace=BCS_BoneSpace
    alphaScaleBiasClamp {bInterpResult=true, interpSpeedDecreasing=2}
    boneToModify {boneName=dyn_torso_box_bott_bk_MSTR}
    rotation {pitch=32}
   }
   Output Pose(Result=Component To Local(ComponentPose=@n13)) {layerGroup=None}
  ]}
 EventGraph {nodes [
   n1 = Event BlueprintUpdateAnimation()
   > n2 = Sequence()
   > Set rigid_body_alpha(rigid_body_alpha:double=MapRangeClamped(Value=@n1.DeltaTimeX, InRangeA=0.016, InRangeB=0.03, OutRangeA=1, OutRangeB=0.4)) -> (Output_Get:double)
   > Set crouch alpha(crouch alpha:double=0) -> (Output_Get:double)
   > K2Node_Composite_0 = BoxFlapsPosFix!(execute) -> (then:Exec) {is=title K2Node_Composite}
   Branch(execute=@n2.then_1, Condition=Get bIsCrouching())
   > n3 = Branch(Condition=Get bIsMoving2D())
   > Set crouch alpha(crouch alpha=1)
   Set crouch alpha(execute=@n3.else, crouch alpha=0.5)
   n4 = Event OnCharacterCustomizationLoaded()
   > Set currentMesh(currentMesh:Skeletal Mesh Component Budgeted Object Reference=Get Array Item with EFort Custom Part Type(Array=@n4.CustomMeshParts)) -> (Output_Get:Skeletal Mesh Component Budgeted Object Reference)
  ]}
 BoxFlapsPosFix {
  path=EventGraph.K2Node_Composite_0.BoxFlapsPosFix
  nodes [
   Inputs()
   n1 = Get bRigidBodyEnabled()
   n2 = Get CurrentLOD()
   > Set isRigidBodyDisabled(isRigidBodyDisabled:bool=OR Boolean(A=NOT Boolean(A=@n1), B="integer > integer"(A=@n2, B=1) {is=title K2Node_CallFunction})) -> (Output_Get:bool)
   n3 = Get currentMesh() -> (currentMesh:Skeletal Mesh Component Budgeted Object Reference)
   n4 = Get isRigidBodyDisabled() -> (isRigidBodyDisabled:bool)
   > Set posFixL(posFixL:double=float * float(
     A=GetAlphaFromReferenceTransformDelta(SkeletalMeshComponent=@n3, BoneName=spine_05, BaseName=spine_01, DeltaComponentType=EulerX, InputRange=(X=5,Y=30), TwistAxis="1,0,0")
     B:bool=@n4
    ) {is=title K2Node_PromotableOperator}) -> (Output_Get:double)
   > Set posFixR(posFixR:double=float * float(
     A:bool=@n4
     B:double=GetAlphaFromReferenceTransformDelta(SkeletalMeshComponent=@n3, BoneName=spine_05, BaseName=spine_01, DeltaComponentType=EulerX, InputRange=(X=-5,Y=-30), TwistAxis="1,0,0")
    )) -> (Output_Get:double)
   > Set posFixB(posFixB:double=float * float(
     A=@n4
     B=GetAlphaFromReferenceTransformDelta(SkeletalMeshComponent=@n3, BoneName=spine_05, BaseName=spine_01, InputRange=(X=-5,Y=-15), TwistAxis="0,0,1")
    )) -> (Output_Get:double)
   > Outputs!(then:Exec)
  ]
 }
]
variables [
 crouch alpha {type=number}
 currentMesh {type=SkeletalMeshComponentBudgeted}
 Custom Base Mesh {type=SkeletalMeshComponent}
 isRigidBodyDisabled {type=boolean}
 posFixB {type=number}
 posFixL {type=number}
 posFixR {type=number}
 rigid_body_alpha {type=number}
]
