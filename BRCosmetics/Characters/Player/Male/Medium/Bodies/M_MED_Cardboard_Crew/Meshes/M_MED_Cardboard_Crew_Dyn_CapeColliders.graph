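# Physics asset of collision shapes for the male medium Cardboard Crew player cosmetic body, named as colliders for dynamic capes.
# Each body is a single capsule on a bone, covering head, neck, spine, pelvis, arms, hands, legs and feet.
# Extra capsules on the back, left and right bottom flap bones of the cardboard torso box add the costume's shape.
# It has no constraints, so it defines collision volumes only, with no jointed ragdoll.
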PhysicsAsset /BRCosmetics/Characters/Player/Male/Medium/Bodies/M_MED_Cardboard_Crew/Meshes/M_MED_Cardboard_Crew_Dyn_CapeColliders
bodies {
 calf_l [{shapeType=Capsule, length=20.75, radius=9.25, shapeName=calf_l_capsule, center=<-21.9403 0.0002 -1.0333>, rotation=<rot -87.3035 0 0>}]
 calf_r [{shapeType=Capsule, length=20.75, radius=9.25, shapeName=calf_r_capsule, center=<22.1142 0.0002 1.0409>, rotation=<rot -87.3051 0 0>}]
 clavicle_l [{
   shapeType=Capsule
   length=10.25
   radius=6.5
   shapeName=clavicle_l_capsule
   center=<7.8408 -0.0001 -0.1877>
   rotation=<rot -88.618 -179.9994 179.9994>
  }]
 clavicle_r [{
   shapeType=Capsule
   length=10.25
   radius=6.5
   shapeName=clavicle_r_capsule
   center=<-8.8662 0.0003 0.2124>
   rotation=<rot -88.6163 -179.9994 179.9994>
  }]
 dyn_torso_box_bott_flap_bk [{shapeType=Capsule, length=22.25, radius=7.75, shapeName=dyn_torso_box_bott_flap_bk_capsule, rotation=<rot -56.4109 0 -96.2959>}]
 dyn_torso_box_bott_flap_l [{shapeType=Capsule, length=23.25, radius=6.75, shapeName=dyn_torso_box_bott_flap_l_capsule, rotation=<rot -56.4093 0 -96.2932>}]
 dyn_torso_box_bott_flap_r [{shapeType=Capsule, length=23.25, radius=6.75, shapeName=dyn_torso_box_bott_flap_r_capsule, rotation=<rot -56.4093 0 -96.2949>}]
 foot_l [{shapeType=Capsule, length=8.5, radius=10.75, shapeName=foot_l_capsule, center=<0 7.6641 -0.3878>, rotation=<rot 0 0 -87.1018>}]
 foot_r [{shapeType=Capsule, length=8.5, radius=10.75, shapeName=foot_r_capsule, center=<0 -7.9542 0.4036>, rotation=<rot 0 0 -87.1018>}]
 hand_l [{shapeType=Capsule, length=7, radius=7.25, shapeName=hand_l_capsule, center=<3.7958 0 1.99>, rotation=<rot -62.3326 0.0002 0>}]
 hand_r [{shapeType=Capsule, length=7, radius=7.25, shapeName=hand_r_capsule, center=<-5.5369 0 -2.9032>, rotation=<rot -62.3298 0.0002 -0.0001>}]
 head [{shapeType=Capsule, length=2.5, radius=19.25, shapeName=head_capsule, center=<14.3538 0.0002 0>}]
 lowerarm_l [{shapeType=Capsule, length=16.25, radius=6, shapeName=lowerarm_l_capsule, center=<9.6542 0 1.2766>, rotation=<rot -82.4675 -0.0002 0.0001>}]
 lowerarm_r [{shapeType=Capsule, length=16.25, radius=6, shapeName=lowerarm_r_capsule, center=<-9.5692 0 -1.2659>, rotation=<rot -82.4655 0.0001 0.0001>}]
 neck_01 [{shapeType=Capsule, length=8.75, radius=7.5, shapeName=neck_01_capsule, rotation=<rot -90 179.9997 -179.9996>}]
 pelvis [{shapeType=Capsule, length=4.6198, radius=12.8664, center=<-1.3506 -1.4738 0.0001>}]
 spine_01 [{shapeType=Capsule, length=0, radius=16.2378, center=<1.0152 -0.4381 0>}]
 spine_02 [{shapeType=Capsule, length=0, radius=16.2057, center=<0.7574 -1.0776 0>}]
 spine_03 [{shapeType=Capsule, length=4.75, radius=17, center=<2.5368 0.1326 0>}]
 spine_04 [{shapeType=Capsule, length=10, radius=15.75, shapeName=spine_04_capsule, center=<0.6241 0 0>}]
 spine_05 [{shapeType=Capsule, length=7.75, radius=16.75, shapeName=spine_05_capsule, center=<2.9018 0 0>}]
 thigh_l [{shapeType=Capsule, length=37.2403, radius=8.0385, center=<-17.728 0 0>, rotation=<rot 87.5182 -179.9999 -179.9999>}]
 thigh_r [{shapeType=Capsule, length=37.2403, radius=8.0385, center=<17.728 0 0>, rotation=<rot 90 0 0>}]
 upperarm_l [{
   shapeType=Capsule
   length=23.25
   radius=6.75
   shapeName=upperarm_l_capsule
   center=<11.4401 0.0009 0.3801>
   rotation=<rot 88.0986 179.9995 179.9994>
  }]
 upperarm_r [{
   shapeType=Capsule
   length=23.25
   radius=6.75
   shapeName=upperarm_r_capsule
   center=<-10.2406 0.0005 -0.3364>
   rotation=<rot 88.1098 179.9995 179.9994>
  }]
}
constraints []
