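# Control Rig for the male medium Cardboard Crew face accessory that procedurally adjusts its head bone and flaps as the character's head rotates.
# The head's rotation relative to spine_05 is converted to Euler, and its Z angle feeds two clamped remaps that nudge dyn_head along local X for either rotation direction.
# The same angle, amplified and clamped mostly to negative values, additively rotates both back flap base bones; front flaps are not driven.
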ControlRigBlueprint /BRCosmetics/Characters/Player/Male/Medium/Bodies/M_MED_Cardboard_Crew/Meshes/Parts/M_MED_Cardboard_Crew_FaceAcc_CtrlRig
graphs [
 RigVMFunctionLibrary {links [], nodes []}
 RigVMModel {nodes [
   n1 = &RigUnit_BeginExecution()
   n2 = &RigVMFunction_MathQuaternionToEuler(RotationOrder=ZYX, Value=&RigUnit_GetRelativeBoneTransform(Bone=head, Space=spine_05).Transform.Rotation)
   /Script/ControlRig.RigUnit_ModifyBoneTransforms(
    BoneToModify=((Bone="dyn_head",Transform=(Rotation=(X=0,Y=0,Z=0,W=1),Translation=(X=0,Y=0,Z=0),Scale3D=(X=1,Y=1,Z=1))))
    Mode=AdditiveLocal
    Weight=1
    WeightMaximum=1
    WeightMinimum=0
    ExecutePin=@n1.ExecutePin
    BoneToModify.0.Transform.Translation.X=&RigVMFunction_MathFloatRemap(SourceMaximum=-48, SourceMinimum=0, TargetMinimum=0, bClamp=true, Value=@n2.Result.Z).Result
   ) {is=struct RigVMUnitNode}
   n3 = &RigUnit_GetRelativeBoneTransform(Bone=head, Space=spine_05)
   > &RigUnit_ModifyBoneTransforms(
    BoneToModify=((Bone="dyn_head",Transform=(Rotation=(X=0,Y=0,Z=0,W=1),Translation=(X=0,Y=0,Z=0),Scale3D=(X=1,Y=1,Z=1))))
    Mode=AdditiveLocal
    Weight=1
    WeightMaximum=1
    WeightMinimum=0
    BoneToModify.0.Transform.Translation.X=&RigVMFunction_MathFloatRemap(
     SourceMaximum=22
     SourceMinimum=0
     TargetMaximum=2
     TargetMinimum=0
     bClamp=true
     Value=&RigVMFunction_MathQuaternionToEuler(RotationOrder=ZYX, Value=@n3.Transform.Rotation).Result.Z
    ).Result
   )
   n4 = &RigVMFunction_MathQuaternionFromEuler(Euler=/Script/RigVM.RigVMFunction_MathVectorClamp(
     Maximum=(X=1,Y=1,Z=1)
     Minimum=(X=0,Y=0,Z=-115)
     Value=/Script/RigVM.RigVMFunction_MathVectorMul(A=(X=0,Y=0,Z=3), B=@n2.Result) {is=struct RigVMUnitNode, template="Multiply::Execute(in A,in B,out Result)"}.Result
    ) {is=struct RigVMUnitNode, template="Clamp::Execute(in Value,in Minimum,in Maximum,out Result)"}.Result)
   > &RigUnit_ModifyBoneTransforms(
    BoneToModify=((Bone="dyn_headFlapBR_01_Base",Transform=(Rotation=(X=0,Y=0,Z=0,W=1),Translation=(X=0,Y=0,Z=0),Scale3D=(X=1,Y=1,Z=1))))
    Mode=AdditiveLocal
    Weight=1
    WeightMaximum=1
    WeightMinimum=0
    BoneToModify.0.Transform.Rotation=@n4.Result
   )
   > &RigUnit_ModifyBoneTransforms(
    BoneToModify=((Bone="dyn_headFlapBL_01_Base",Transform=(Rotation=(X=0,Y=0,Z=0,W=1),Translation=(X=0,Y=0,Z=0),Scale3D=(X=1,Y=1,Z=1))))
    Mode=AdditiveLocal
    Weight=1
    WeightMaximum=1
    WeightMinimum=0
    BoneToModify.0.Transform.Rotation=@n4.Result
   )
  ]}
]
hierarchy [
 root {loc=<0 0 0>, rot=<0 0 0 1>}
 pelvis {parent=root, loc=<0 -0.7071 93.1191>, rot=<0.0127 0.707 -0.0127 -0.707>}
 spine_01 {parent=pelvis, loc=<6.7494 -0.6388 0>, rot=<0 0 0.034 -0.9994>}
 spine_02 {parent=spine_01, loc=<7.2518 -0.035 0>, rot=<0 0 0.0238 -0.9997>}
 spine_03 {parent=spine_02, loc=<6.0104 0 0>, rot=<0 0 0.0149 -0.9999>}
 spine_04 {parent=spine_03, loc=<8.1024 0 0>, rot=<0 0 -0.0132 -0.9999>}
 spine_05 {parent=spine_04, loc=<10.0146 0 0>, rot=<0 0 -0.0015 -1>}
 clavicle_l {parent=spine_05, loc=<4.3575 -0.1466 -2.3223>, rot=<0.0465 -0.7289 -0.009 -0.6829>}
 upperarm_l {parent=clavicle_l, loc=<15.5111 0 0>, rot=<0.0226 -0.3234 -0.0088 -0.9459>}
 clavicle_r {parent=spine_05, loc=<4.3575 -0.1466 2.3223>, rot=<0.7289 0.0465 -0.6829 0.009>}
 upperarm_r {parent=clavicle_r, loc=<-15.5111 0 0>, rot=<0.0226 -0.3234 -0.0088 -0.9459>}
 neck_01 {parent=spine_05, loc=<11.3138 0.1529 0>, rot=<0 0 -0.2003 -0.9797>}
 neck_02 {parent=neck_01, loc=<4.5627 0 0>, rot=<0 0 0.0504 -0.9987>}
 head {parent=neck_02, loc=<5.3914 0 0>, rot=<0 0 0.115 -0.9934>}
 faceAttach {parent=head, loc=<-0.8266 -0.1832 0>, rot=<0.0044 -0.3827 -0.0044 -0.9239>, scale=<0.89 0.89 0.89>}
 C_jaw {parent=faceAttach, loc=<0.3123 3.2483 0.3583>, rot=<0.0044 -0.3827 -0.0044 -0.9239>}
 dyn_head {parent=head, loc=<0 0 0>, rot=<0 0 0 -1>}
 dyn_headFlapBR_01_Base {parent=dyn_head, loc=<-0.7676 -3.6468 6.3926>, rot=<0.0162 -0.8782 0.4775 -0.021>}
 dyn_headFlapBR_01 {parent=dyn_headFlapBR_01_Base, loc=<0 0 0>, rot=<0.0071 0.0039 0.0032 -1>}
 dyn_headFlapBR_02 {parent=dyn_headFlapBR_01, loc=<6.1688 -0.2818 -0.2533>, rot=<-0.0028 -0.0082 0.8304 -0.5571>}
 dyn_headFlapBR_03 {parent=dyn_headFlapBR_02, loc=<6.1256 0.0302 0.7926>, rot=<0.0157 0.0643 -0.0014 -0.9978>}
 dyn_headFlapBL_01_Base {parent=dyn_head, loc=<-0.6563 -5.7509 -4.2315>, rot=<0.0257 -0.9444 -0.3275 0.0169>}
 dyn_headFlapBL_01 {parent=dyn_headFlapBL_01_Base, loc=<0 0 0>, rot=<0 0 0 -1>}
 dyn_headFlapBL_02 {parent=dyn_headFlapBL_01, loc=<6.3202 -0.0062 -0.2235>, rot=<-0.0207 -0.027 0.8226 -0.5676>}
 dyn_headFlapBL_03 {parent=dyn_headFlapBL_02, loc=<5.6754 -0.0094 0.172>, rot=<0.0006 0.0152 0.0008 -0.9999>}
 dyn_headFlapFL_01_Base {parent=dyn_head, loc=<-1.4466 6.8608 -6.4665>, rot=<-0.0125 0.4679 0.8831 -0.0324>}
 dyn_headFlapFL_01 {parent=dyn_headFlapFL_01_Base, loc=<0 0 0>, rot=<0 0 0 -1>}
 dyn_headFlapFL_02 {parent=dyn_headFlapFL_01, loc=<7.0207 -0.1803 0.4995>, rot=<-0.0103 -0.0058 0.7556 -0.6549>}
 dyn_headFlapFL_03 {parent=dyn_headFlapFL_02, loc=<6.4337 -0.0246 0.2398>, rot=<0.0009 0.016 0.0019 -0.9999>}
 dyn_headFlapFR_01_Base {parent=dyn_head, loc=<-1.7119 8.4262 4.692>, rot=<0.0486 -0.285 0.9557 -0.0547>}
 dyn_headFlapFR_01 {parent=dyn_headFlapFR_01_Base, loc=<0 0 0>, rot=<0.0502 -0.0394 -0.002 -0.998>}
 dyn_headFlapFR_02 {parent=dyn_headFlapFR_01, loc=<6.742 -0.2364 -0.276>, rot=<-0.0046 -0.004 0.7616 -0.648>}
 dyn_headFlapFR_03 {parent=dyn_headFlapFR_02, loc=<6.0553 -0.1782 -0.8685>, rot=<0.0479 -0.0718 0.0112 -0.9962>}
]
